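; Base quest prop actor whose readiness logic is left to children, kept simple for a time trial in 18.40.
; Consolidated quest component ready and not-ready events forward the player controller to empty child events.
; HandleQuestReady, not called here, checks whether the component's objective is active and unhides the actor per controller in splitscreen.
; Interaction is always allowed, and construction gives the static mesh a trigger-only interaction collision profile.
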(event EventBeginPlay)

(event Collision|EventActorBeginOverlap
  (params (OtherActor :type Actor%20Object%20Reference)))

(event EventTick
  (params (DeltaSeconds :type Float%20%28single-precision%29)))

(event Custom|OnQuestReady_Child
  (params (PlayerController :type Fort%20Player%20Controller%20Object%20Reference)))

(event OnQuestReady(ConsolidatedQuestComponent)
  (params (PlayerController :type Fort%20Player%20Controller%20Object%20Reference))
  ;[1 1 1] We can handle all the nice templated abstract stuff later, but let's keep it safe in 18.40 for the main time trial we're supporting
  (CallFunction|OnQuestReadyChild PlayerController))

(event OnQuestNotReady(ConsolidatedQuestComponent)
  (params (PlayerController :type Fort%20Player%20Controller%20Object%20Reference))
  ;[1 1 1] We can handle all the nice templated abstract stuff later, but let's keep it safe in 18.40 for the main time trial we're supporting
  (CallFunction|OnQuestNotReadyChild PlayerController))

(event Custom|OnQuestNotReady_Child
  (params (PlayerController :type Fort%20Player%20Controller%20Object%20Reference)))

(fn BlueprintCanInteract
  (params (InteractingPawn :type Fort%20Pawn%20Object%20Reference) (InteractionBeingAttempted :type EInteractionBeingAttempted%20Enum :default "FirstInteraction") (InteractionType :type TInteractionType%20Enum :default "IT_NoInteraction"))
  (return true))

(fn HandleQuestReady
  (params (InPlayerController :type Fort%20Player%20Controller%20Object%20Reference))
  (bind _output_get (LocalVariables|SetLocal_InPC InPlayerController))
  (if (Quests|IsQuestObjectiveActive (Quests|GetQuestManager _output_get "Invalid") (Variables|/Script/FortniteGame.ConsolidatedQuestComponent|GetObjectiveBackendName (Variables|BuildingGameplayActorPropQuest|GetConsolidatedQuestComponent)))
    (if (PlayerController|IsinSplitscreen (LocalVariables|GetLocal_InPC))
      (PlayerController|SetActorHiddenControllerOnly (LocalVariables|GetLocal_InPC) self false)
      (else
        _))
    (else
      (return (LocalVariables|GetLocal_InPC) false))))

(fn ConstructionScript
  (Collision|SetCollisionProfileName (Variables|Default|GetStaticMesh) "FortTriggerOnlyInteractions" true :!class /Script/Engine.PrimitiveComponent))
